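# Physics asset for the Comet Male pickaxe skeletal mesh, providing bodies and joints for two dynamic feather groups, A and B, on the two-handed handle.
# Each group links a small box base to the handle with free angular motion, then hinges its first feather capsule to that base on one limited swing axis.
# The second and third feather capsules both branch from the first with tight swing and twist limits, rather than forming a serial chain.
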PhysicsAsset /BRCosmetics/Weapons/FORT_Melee/Pickaxe_Comet_Male/Meshes/SK_Pickaxe_Comet_Male_Physics
bodies {
 Melee_TwoHanded_Handle [{shapeType=Capsule, length=71.1043, radius=14.9026, center=<0.7328 -1.3074 30.0257>, rotation=<rot -0.0818 -90 0.0749>}]
 dyn_featherMainA_01 [{shapeType=Capsule, length=16.1247, radius=10.8247, center=<12 0 0>, rotation=<rot -0.0001 -112.501 89.9999>}]
 dyn_featherMainA_01_Base [{shapeType=Box, extentX=1.01, extentY=1.01, extentZ=1.01, rotation=<rot 0 -90 0>}]
 dyn_featherMainA_02 [{shapeType=Capsule, length=9.0029, radius=7.2402, center=<10 0.5917 0.0978>, rotation=<rot 0 82.4331 -89.2824>}]
 dyn_featherMainA_03 [{shapeType=Capsule, length=7.9468, radius=4.3961, center=<3 0 0>, rotation=<rot 0 71.7032 -90>}]
 dyn_featherMainB_01 [{shapeType=Capsule, length=16.7212, radius=10.8168, center=<12 0 0>, rotation=<rot -0.0001 -112.501 89.9999>}]
 dyn_featherMainB_01_Base [{shapeType=Box, extentX=1.01, extentY=1.01, extentZ=1.01, rotation=<rot 0 -90 0>}]
 dyn_featherMainB_02 [{shapeType=Capsule, length=9.6487, radius=7.7824, center=<10 0.5403 0.1514>, rotation=<rot 0 91.7216 -89.2015>}]
 dyn_featherMainB_03 [{shapeType=Capsule, length=8.6253, radius=4.3947, center=<3 0 0>, rotation=<rot 0 71.5374 -88.3974>}]
}
constraints [
 {
  bone1Name=dyn_featherMainA_01
  bone2Name=dyn_featherMainA_01_Base
  swing1LimitDegrees=20
  swing2LimitDegrees=45
  swing2Motion=Locked
  twistLimitDegrees=45
  twistMotion=Locked
 }
 {
  bone1Name=dyn_featherMainB_01
  bone2Name=dyn_featherMainB_01_Base
  swing1LimitDegrees=20
  swing2LimitDegrees=45
  swing2Motion=Locked
  twistLimitDegrees=45
  twistMotion=Locked
 }
 {
  bone1Name=dyn_featherMainA_01_Base
  bone2Name=Melee_TwoHanded_Handle
  swing1LimitDegrees=45
  swing1Motion=Free
  swing2LimitDegrees=45
  swing2Motion=Free
  twistLimitDegrees=45
  twistMotion=Free
 }
 {
  bone1Name=dyn_featherMainB_01_Base
  bone2Name=Melee_TwoHanded_Handle
  swing1LimitDegrees=45
  swing1Motion=Free
  swing2LimitDegrees=45
  swing2Motion=Free
  twistLimitDegrees=45
  twistMotion=Free
 }
 {bone1Name=dyn_featherMainA_02, bone2Name=dyn_featherMainA_01, swing1LimitDegrees=25, swing2LimitDegrees=8, twistLimitDegrees=6}
 {bone1Name=dyn_featherMainA_03, bone2Name=dyn_featherMainA_01, swing1LimitDegrees=25, swing2LimitDegrees=8, twistLimitDegrees=6}
 {bone1Name=dyn_featherMainB_02, bone2Name=dyn_featherMainB_01, swing1LimitDegrees=25, swing2LimitDegrees=8, twistLimitDegrees=6}
 {bone1Name=dyn_featherMainB_03, bone2Name=dyn_featherMainB_01, swing1LimitDegrees=25, swing2LimitDegrees=8, twistLimitDegrees=6}
]
